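# Material function in the Core UI library that outputs a "Backplate A" color, apparently derived from a blurred Slate post buffer.
# A Post Buffer Blur input defaults to GetSlatePost0.
# Contrast-preserved color is combined with a perceived-brightness three-point ramp via MultiplyAdd, then screen-blended over a second such ramp to lift darker tones.
# HSV-adjustment and max-minus-min power branches end unconnected to the output.
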MaterialFunction /FortUILibrary/MaterialFunctions/CoreUI/MF_UI_CoreUI_BackplateA
graph {
 properties {libraryCategoriesText [UI Library]}
 nodes [
  n1 = Max(A=Saturate(None=NamedRerouteUsage())) {constB=0.001}
  n2 = Constant()
  n3 = NamedRerouteDeclaration(None=Saturate(None=MaterialFunctionCall!(Color=@n1, Contrast=@n2) {materialFunction=/Engine/Functions/Engine_MaterialFunctions01/ImageAdjustment/Contrast_Preserve_Color}.Result)) {name=Preserved Saturated Colors, nodeColor=<rgba 0 0 0 1>}
  n4 = NamedRerouteUsage()
  n5 = NamedRerouteDeclaration(None=Desaturation(None=Saturate(None=@n4))) {name=Perceived Brightness, nodeColor=<rgba 0 0 0 1>}
  n6 = MaterialFunctionCall(A=Constant() {r=0.9}, B=Constant() {r=0.5}, C=Constant(), Alpha=@n5) {materialFunction=&Lerp_3Color}
  n7 = MaterialFunctionCall!(
   Blend=Saturate(None=MaterialFunctionCall!(Base=@n3, Add=@n6.Result) {materialFunction=/Engine/Functions/Engine_MaterialFunctions02/Math/MultiplyAdd}.Result)
   Base=MaterialFunctionCall(A=Constant() {r=0.4}, B=Constant() {r=0.4}, C=Constant(), Alpha=@n5) {materialFunction=&Lerp_3Color}.Result
  ) {materialFunction=/Engine/Functions/Engine_MaterialFunctions03/Blends/Blend_Screen}
  FunctionOutput(None=NamedRerouteDeclaration(None=Saturate(None=@n7.Result)) {name=Backplate A, nodeColor=<rgba 0 0 0 1>})
  n8 = NamedRerouteUsage()
  n9 = RgbToHsv!(None=NamedRerouteDeclaration(None=FunctionInput(Preview=@n8) {bUsePreviewValueAsDefault=true, inputName=Post Buffer Blur}) {name=Buffer RGB, nodeColor=<rgba 0 0 0 1>})
  n10 = ComponentMask(None=@n9) {r=true}
  n11 = ComponentMask(None=@n9) {g=true}
  n12 = MaterialFunctionCall(A=@n11, B=LinearInterpolate(A=@n11) {constA=0.2, constAlpha=0.2}, C=@n11, Alpha=@n11) {materialFunction=&Lerp_3Color}
  n13 = LinearInterpolate(A=ComponentMask(None=@n9) {b=true}) {constAlpha=0.6}
  n14 = Saturate(None=HsvToRgb!(None=MaterialFunctionCall(R=@n10, G=@n12.Result, B=@n13) {materialFunction=&AppendMany}.RGB))
  LinearInterpolate(
   A=@n14
   B=MaterialFunctionCall(Blend=@n14, Base=Constant3Vector() {constant=<rgba 0.6 0.6 0.6 0>}) {materialFunction=&Blend_Screen}.Result
  ) {constAlpha=0.75}
  NamedRerouteDeclaration(None=MaterialFunctionCall!("UVs (Optional)") {materialFunction=/Engine/Functions/UserInterface/GetSlatePost0}.LinearRGB) {name=Blur RGB, nodeColor=<rgba 0 0 0 1>}
  n15 = NamedRerouteUsage()
  n16 = LinearInterpolate(A=Saturate(None=NamedRerouteUsage())) {constAlpha=0.6}
  n17 = NamedRerouteUsage()
  n18 = ComponentMask(None=@n17) {r=true}
  n19 = ComponentMask(None=@n17) {g=true}
  n20 = ComponentMask(None=@n17) {b=true}
  n21 = Max(A=Max(A=@n18, B=@n19), B=@n20)
  n22 = Min(A=@n18, B=Min(A=@n19, B=@n20))
  n23 = NamedRerouteDeclaration(None=Saturate(None=Subtract(A=@n21, B=@n22) {desc=Take Brightness Value})) {name=Brightness, nodeColor=<rgba 0 0 0 1>}
  n24 = Saturate(None=Power(Base=Multiply(A=@n15, B=@n16), Exp=LinearInterpolate(Alpha=@n23) {constA=0.9, constB=0.6}) {constExponent=0.6})
  n25 = Constant3Vector() {constant=<rgba 0.6 0.6 0.6 0>}
  Saturate(None=LinearInterpolate(
    A=@n24
    B=MaterialFunctionCall(Blend=@n24, Base=@n25) {materialFunction=&Blend_Screen}.Result
   ) {constAlpha=0.75})
 ]
}
