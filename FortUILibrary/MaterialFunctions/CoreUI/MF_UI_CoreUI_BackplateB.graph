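# UI material function that builds the Core UI "Backplate B" color treatment, with a main output and a "Weak" output, both saturated.
# A Post Buffer Blur input offers a preview default, alongside a GetSlatePost0 blur sample.
# The main output lifts values with a three-point lerp; Weak blends two values.
# Other branches, ending unconnected, adjust color through HSV, screen blends, color-preserving contrast and brightness-driven lerps that whiten darks.
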MaterialFunction /FortUILibrary/MaterialFunctions/CoreUI/MF_UI_CoreUI_BackplateB
graph {
 properties {previewMaterialDomain=MD_UI, libraryCategoriesText [UI Library]}
 nodes [
  n1 = NamedRerouteUsage()
  n2 = Constant() {r=0.05}
  n3 = Constant() {r=0.1}
  n4 = Constant() {r=0.375}
  n5 = NamedRerouteUsage()
  n6 = Add(
   A=@n1
   B=Multiply(
    A=OneMinus(None=@n1)
    B=MaterialFunctionCall!(1=@n2, 2=@n3, 3=@n4, 4, A=@n5) {materialFunction=/Engine/Functions/Engine_MaterialFunctions02/Lerp_Multiple_Float}.Lerp 3 Inputs
   ) {constB=0.05}
  )
  FunctionOutput(None=NamedRerouteDeclaration(None=Saturate(None=@n6)) {name=Backplate B, nodeColor=<rgba 0 0 0 1>})
  n7 = NamedRerouteUsage()
  n8 = NamedRerouteUsage()
  FunctionOutput(None=Saturate(None=LinearInterpolate(A=@n7, B=@n8) {constAlpha=0.4})) {outputName=Weak}
  NamedRerouteDeclaration(None=MaterialFunctionCall!("UVs (Optional)") {materialFunction=/Engine/Functions/UserInterface/GetSlatePost0}.LinearRGB) {name=Blur RGB, nodeColor=<rgba 0 0 0 1>}
  n9 = NamedRerouteUsage()
  n10 = RgbToHsv!(None=NamedRerouteDeclaration(None=FunctionInput(Preview=@n9) {bUsePreviewValueAsDefault=true, inputName=Post Buffer Blur}) {name=Buffer RGB, nodeColor=<rgba 0 0 0 1>})
  n11 = ComponentMask(None=@n10) {r=true}
  n12 = ComponentMask(None=@n10) {g=true}
  n13 = MaterialFunctionCall(1=@n12, 2=LinearInterpolate(A=@n12) {constA=0.2, constAlpha=0.1}, 3=@n12, A=@n12) {materialFunction=&Lerp_Multiple_Float}
  n14 = NamedRerouteDeclaration(None=LinearInterpolate(A=ComponentMask(None=@n10) {b=true}) {constAlpha=0.2}) {name=Brightness Value, nodeColor=<rgba 0 0 0 1>}
  n15 = Saturate(None=HsvToRgb!(None=MaterialFunctionCall(R=@n11, G=@n13.Lerp 3 Inputs, B=@n14) {materialFunction=&AppendMany}.RGB))
  n16 = Constant3Vector() {constant=<rgba 0.7135 0.7135 0.7135 1>}
  Saturate(None=LinearInterpolate(
    A=@n15
    B=MaterialFunctionCall!(Blend=@n15, Base=@n16) {materialFunction=/Engine/Functions/Engine_MaterialFunctions03/Blends/Blend_Screen}.Result
   ) {constAlpha=0.125})
  n17 = NamedRerouteUsage()
  n18 = LinearInterpolate(A=@n17) {constAlpha=0.3}
  n19 = LinearInterpolate(A=@n17) {constAlpha=0.6}
  n20 = LinearInterpolate(B=@n17) {constAlpha=0.75}
  n21 = Saturate(None=Multiply(
    A=NamedRerouteUsage()
    B=MaterialFunctionCall(1=@n18, 2=@n19, 3=@n20, A=@n17) {materialFunction=&Lerp_Multiple_Float}.Lerp 3 Inputs
   ))
  n22 = Constant3Vector() {constant=<rgba 0.7135 0.7135 0.7135 1>}
  Saturate(None=LinearInterpolate(
    A=@n21
    B=MaterialFunctionCall(Blend=@n21, Base=@n22) {materialFunction=&Blend_Screen}.Result
   ) {constAlpha=0.125})
  n23 = Saturate(None=NamedRerouteUsage())
  n24 = Saturate(None=MaterialFunctionCall!(Color=Max(A=@n23) {constB=0.001}, Contrast=Constant()) {materialFunction=/Engine/Functions/Engine_MaterialFunctions01/ImageAdjustment/Contrast_Preserve_Color}.Result)
  n25 = Desaturation(None=NamedRerouteUsage())
  n26 = NamedRerouteUsage()
  n27 = ComponentMask(None=@n26) {r=true}
  n28 = ComponentMask(None=@n26) {g=true}
  n29 = ComponentMask(None=@n26) {b=true}
  n30 = Max(A=Max(A=@n27, B=@n28), B=@n29)
  n31 = Min(A=@n27, B=Min(A=@n28, B=@n29))
  n32 = NamedRerouteDeclaration(None=Saturate(None=Subtract(A=@n30, B=@n31) {desc=Take Brightness Value})) {name=Brightness, nodeColor=<rgba 0 0 0 1>}
  n33 = MaterialFunctionCall(
   1=Constant() {r=0.125}
   2=Constant() {r=0.525}
   3=Constant() {r=0.825}
   A=MaterialFunctionCall(1=NamedRerouteDeclaration(None=@n25) {name=Perceived Brightness, nodeColor=<rgba 0 0 0 1>}, 2=@n32, 3=@n32, A=@n32) {materialFunction=&Lerp_Multiple_Float}.Lerp 3 Inputs
  ) {materialFunction=&Lerp_Multiple_Float}
  n34 = Saturate(None=MaterialFunctionCall!(
    Base=NamedRerouteDeclaration(None=@n24) {name=Preserved Saturated Colors, nodeColor=<rgba 0 0 0 1>}
    Add=LinearInterpolate(Alpha=@n33.Lerp 3 Inputs) {constA=-1}
   ) {materialFunction=/Engine/Functions/Engine_MaterialFunctions02/Math/MultiplyAdd}.Result)
  n35 = Constant() {r=0.1}
  n36 = Constant() {r=0.05}
  n37 = Constant() {r=0.2}
  n38 = NamedRerouteUsage()
  Saturate(None=LinearInterpolate(
    A=NamedRerouteDeclaration(None=@n34) {name=Adjusted colors, nodeColor=<rgba 0 0 0 1>}
    Alpha=MaterialFunctionCall(1=@n35, 2=@n36, 3=@n37, A=@n38) {desc=Add white values on dark values, materialFunction=&Lerp_Multiple_Float}.Lerp 3 Inputs
   ) {constAlpha=0.075})
 ]
}
