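# CoreUI backplate function that turns a blurred UI post-buffer color into a darkened backplate color.
# The output blends the color toward a near-black tinted multiply, with desaturated luminosity driving tint and strength so brighter areas darken more.
# The "Post Buffer Blur" input previews and defaults to a reroute, with GetSlatePost0 declared as Blur RGB.
# HSV remap and contrast-based darkening branches end unconnected to the output.
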MaterialFunction /FortUILibrary/MaterialFunctions/CoreUI/MF_UI_CoreUI_BackplateC
graph {
 properties {libraryCategoriesText [UI Library]}
 nodes [
  Desaturation(None=Saturate(None=NamedRerouteUsage()))
  n1 = Saturate(None=NamedRerouteUsage())
  n2 = Constant3Vector() {constant=<rgba 0.0513 0.0513 0.0513 1>}
  n3 = Constant3Vector()
  n4 = Saturate(None=Desaturation(None=NamedRerouteUsage()) {desc=Gives better result on dark background})
  n5 = LinearInterpolate(
   A=@n1
   B=Multiply(A=@n1, B=LinearInterpolate(A=@n2, B=@n3, Alpha=@n4) {constA=0.1, constB=0.1})
   Alpha=LinearInterpolate(Alpha=@n4) {constA=0.5, constB=0.9, desc=Controls darkening based on luminosity}
  )
  FunctionOutput(None=NamedRerouteDeclaration(None=Saturate(None=@n5)) {name=Backplate C, nodeColor=<rgba 0 0 0 1>})
  n6 = NamedRerouteUsage()
  n7 = MaterialFunctionCall!(Color=Max(A=Saturate(None=@n6)) {constB=0.001}, Contrast=Constant()) {materialFunction=/Engine/Functions/Engine_MaterialFunctions01/ImageAdjustment/Contrast_Preserve_Color}
  n8 = NamedRerouteUsage()
  n9 = ComponentMask(None=@n8) {r=true}
  n10 = ComponentMask(None=@n8) {g=true}
  n11 = ComponentMask(None=@n8) {b=true}
  n12 = Max(A=Max(A=@n9, B=@n10), B=@n11)
  n13 = Min(A=@n9, B=Min(A=@n10, B=@n11))
  n14 = NamedRerouteDeclaration(None=Saturate(None=Subtract(A=@n12, B=@n13) {desc=Take Brightness Value})) {name=Brightness, nodeColor=<rgba 0 0 0 1>}
  n15 = LinearInterpolate(A=@n14) {constA=0.2, constAlpha=0.1}
  n16 = LinearInterpolate(B=@n14) {constAlpha=0.4}
  n17 = LinearInterpolate(B=@n14) {constAlpha=0.75}
  n18 = Saturate(None=Multiply(
    A=Saturate(None=@n7.Result)
    B=MaterialFunctionCall(A=@n15, B=@n16, C=@n17, Alpha=@n14) {materialFunction=&Lerp_3Color}.Result
   ))
  LinearInterpolate(A=@n18, B=Multiply(A=@n18, B=Constant3Vector() {constant=<rgba 0.0513 0.0513 0.0513 1>}))
  n19 = NamedRerouteUsage()
  n20 = RgbToHsv!(None=NamedRerouteDeclaration(None=FunctionInput(Preview=@n19) {bUsePreviewValueAsDefault=true, inputName=Post Buffer Blur}) {name=Buffer RGB, nodeColor=<rgba 0 0 0 1>})
  n21 = ComponentMask(None=@n20) {g=true}
  n22 = ComponentMask(None=@n20) {b=true}
  n23 = MaterialFunctionCall(
   R=ComponentMask(None=@n20) {r=true}
   G=MaterialFunctionCall(A=@n21, B=LinearInterpolate(A=@n21) {constA=0.2, constAlpha=0.1}, C=@n21, Alpha=@n21) {materialFunction=&Lerp_3Color}.Result
   B=MaterialFunctionCall(
    A=LinearInterpolate(A=@n22) {constA=0.2, constAlpha=0.2}
    B=LinearInterpolate(B=@n22) {constAlpha=0.9, constB=0.5}
    C=LinearInterpolate(B=@n22) {constAlpha=0.8}
    Alpha=@n22
   ) {materialFunction=&Lerp_3Color}.Result
  ) {materialFunction=&AppendMany}
  Multiply(A=Saturate(None=HsvToRgb!(None=@n23.RGB)), B=Constant() {r=0.25})
  NamedRerouteDeclaration(None=MaterialFunctionCall!("UVs (Optional)") {materialFunction=/Engine/Functions/UserInterface/GetSlatePost0}.LinearRGB) {name=Blur RGB, nodeColor=<rgba 0 0 0 1>}
 ]
}
