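# Material function in the CoreUI library that produces Backplate D, a darker variant of the Backplate C UI backplate.
# It feeds the Post Buffer Blur input into MF_UI_CoreUI_BackplateC, then blends that result mostly toward a copy multiplied by a near-black, blue-violet tint.
# The input's preview default draws on the Slate post buffer's linear RGB through a named reroute.
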MaterialFunction /FortUILibrary/MaterialFunctions/CoreUI/MF_UI_CoreUI_BackplateD
graph {
 properties {libraryCategoriesText [UI Library]}
 nodes [
  n1 = MaterialFunctionCall!(Post Buffer Blur=FunctionInput(Preview=NamedRerouteUsage()) {bUsePreviewValueAsDefault=true, inputName=Post Buffer Blur}) {materialFunction=&/MF_UI_CoreUI_BackplateC}
  n2 = Multiply(A=@n1.Result, B=Constant3Vector() {constant=<rgba 0.0048 0.004 0.0116 1>})
  n3 = Constant() {r=0.7}
  FunctionOutput(None=NamedRerouteDeclaration(None=LinearInterpolate(A=@n1.Result, B=@n2, Alpha=@n3) {constAlpha=0.3}) {name=Backplate D, nodeColor=<rgba 0 0 0 1>})
  NamedRerouteDeclaration(None=MaterialFunctionCall!("UVs (Optional)") {materialFunction=/Engine/Functions/UserInterface/GetSlatePost0}.LinearRGB) {name=Blur RGB, nodeColor=<rgba 0 0 0 1>}
 ]
}
