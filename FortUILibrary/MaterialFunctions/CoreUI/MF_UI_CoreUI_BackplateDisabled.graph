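# Material function in the Core UI library that produces the disabled-state look for backplates, built on a blurred Slate post buffer.
# The optional blur input defaults to GetSlatePost0 linear RGB.
# The output blends that color with MF_UI_CoreUI_BackplateB, pulls it toward flat grey, then partially mixes it back over the original buffer.
# A second Color Dodge and desaturation branch is present but does not feed the output.
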MaterialFunction /FortUILibrary/MaterialFunctions/CoreUI/MF_UI_CoreUI_BackplateDisabled
graph {
 properties {libraryCategoriesText [UI Library]}
 nodes [
  n1 = NamedRerouteUsage()
  n2 = LinearInterpolate(
   A=@n1
   B=MaterialFunctionCall!(Post Buffer Blur=@n1) {materialFunction=&/MF_UI_CoreUI_BackplateB}.Result
  )
  n3 = Constant3Vector() {constant=<rgba 0.0513 0.0513 0.0513 1>}
  Desaturation(
   None=LinearInterpolate(
    A=@n2
    B=MaterialFunctionCall!(Blend=@n2, Base=@n3) {materialFunction=/Engine/Functions/Engine_MaterialFunctions03/Blends/Blend_ColorDodge}.Result
   ) {constAlpha=0.125}
   Fraction=Constant() {r=0.2}
  )
  n4 = NamedRerouteUsage()
  n5 = LinearInterpolate(
   A=LinearInterpolate(
    A=@n4
    B=MaterialFunctionCall(Post Buffer Blur=@n4) {materialFunction=&MF_UI_CoreUI_BackplateB}.Result
   ) {constAlpha=0.75}
   B=Constant3Vector() {constant=<rgba 0.521 0.521 0.521 1>}
  )
  FunctionOutput(None=NamedRerouteDeclaration(None=LinearInterpolate(A=@n4, B=@n5) {constAlpha=0.3333}) {name=Backplate Disabled, nodeColor=<rgba 0 0 0 1>})
  n6 = MaterialFunctionCall!("UVs (Optional)") {materialFunction=/Engine/Functions/UserInterface/GetSlatePost0}
  NamedRerouteDeclaration(None=FunctionInput(Preview=NamedRerouteDeclaration(None=@n6.LinearRGB) {name=Blur RGB, nodeColor=<rgba 0 0 0 1>}) {bUsePreviewValueAsDefault=true, inputName=Post Buffer Blur}) {name=Buffer RGB, nodeColor=<rgba 0 0 0 1>}
 ]
}
