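# UI material function that outputs an animated rainbow color varying along U and oscillating over time.
# Primary UV, defaulting to normalized UI UV, blends to secondary UV by UVMix; only the result's X component is used.
# That value plus per-channel phase offsets and Time feeds a cosine palette, remapped to 0-1.
# The PixelSize input is multiplied then divided out, so it effectively cancels.
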MaterialFunction /FortUILibrary/MaterialFunctions/CoreUI/MF_UI_CoreUI_OscillatingColor
graph {nodes [
  n1 = Constant() {r=0.5}
  n2 = MaterialFunctionCall() {materialFunction=&GetUserInterfaceUV}
  n3 = FunctionInput(Preview=@n2.Pixel Size) {bUsePreviewValueAsDefault=true, inputName=PixelSize, inputType=FunctionInput_Vector2, sortPriority=3}
  n4 = LinearInterpolate(
   A=FunctionInput(Preview=@n2.Normalized UV) {bUsePreviewValueAsDefault=true, inputName=PrimaryUV, inputType=FunctionInput_Vector2}
   B=FunctionInput() {bUsePreviewValueAsDefault=true, inputName=SecondaryUV, inputType=FunctionInput_Vector2, sortPriority=1}
   Alpha=FunctionInput() {bUsePreviewValueAsDefault=true, inputName=UVMix, inputType=FunctionInput_Scalar, sortPriority=2}
  )
  n5 = MaterialFunctionCall(Float2=Divide(A=Multiply(A=@n3, B=@n4), B=@n3)) {materialFunction=&BreakOutFloat2Components}
  n6 = Add(
   A=MaterialFunctionCall(X=@n5.R, Y=@n5.R, Z=@n5.R) {materialFunction=&MakeFloat3}.Result
   B=Constant3Vector() {constant=<rgba 0 2 4 0>}
  )
  n7 = Time()
  n8 = Multiply(A=@n1, B=Cosine(None=Add(A=@n6, B=@n7)) {desc=2 PI, period=6.2832})
  FunctionOutput(None=Saturate(None=Add(A=@n1, B=@n8)))
 ]}
