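# UI material function that generates an anti-aliased rounded-rectangle mask and outline following Core UI rules.
# Per-corner radii, given as ratios or pixels, are clamped, optionally scaled by the box's shorter side, then fed to MF_UI_Box for an SDF.
# MF_UI_ShapeOutline derives the outline from that SDF, and both are output as SDFs and masks, alongside UVs.
# Helpers supply aspect-corrected horizontal skew and a viewport scale factor.
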MaterialFunction /FortUILibrary/MaterialFunctions/CoreUI/MF_UI_CoreUI_RoundedMask
graph {
 properties {
  bExposeToLibrary=true
  description=Generate a rounded corner shape based on Core UI rules
  previewMaterialDomain=MD_UI
  libraryCategoriesText [UI Library]
 }
 nodes [
  n1 = NamedRerouteUsage()
  n2 = FunctionInput() {
   bUsePreviewValueAsDefault=true
   description=Defines the size
   inputName=Size
   inputType=FunctionInput_Vector2
   sortPriority=1
   previewValue=<1 1 0 0>
  }
  n3 = FunctionInput() {
   bUsePreviewValueAsDefault=true
   description=Set the corners radius
   inputName=Corner Radius
   inputType=FunctionInput_Vector4
   sortPriority=2
   previewValue=<0.3 0.3 0.3 0.3>
  }
  n4 = NamedRerouteDeclaration(None=@n3) {name=Corner Radius, nodeColor=<rgba 0 0 0 1>}
  n5 = NamedRerouteUsage()
  n6 = FunctionInput() {
   bUsePreviewValueAsDefault=true
   description=Set if you desire to calculate in pixels instead of a ratio/percentage
   inputName=Radius in Pixels
   inputType=FunctionInput_Scalar
   sortPriority=3
   previewValue=<0 0 0 0>
  }
  n7 = Floor(None=@n6)
  n8 = Multiply(A=LinearInterpolate(A=@n4, B=Divide(A=@n4, B=@n5), Alpha=@n7)) {constB=2}
  n9 = LinearInterpolate(B=NamedRerouteUsage(), Alpha=@n7) {constA=1}
  n10 = NamedRerouteDeclaration(None=Min(A=Multiply(A=@n8, B=@n9)) {desc=Avoid rounded corners to exceed the maximum amount}) {name=Remapped Corner Radius, nodeColor=<rgba 0 0 0 1>}
  n11 = NamedRerouteUsage()
  n12 = Min(A=ComponentMask(None=@n11) {r=true}, B=ComponentMask(None=@n11) {g=true})
  n13 = FunctionInput() {
   bUsePreviewValueAsDefault=true
   description=Define if corner radius should be influenced by the size of the box
   inputName=Corners Modulated by Size
   inputType=FunctionInput_Scalar
   sortPriority=40
   previewValue=<1 1 1 1>
  }
  n14 = LinearInterpolate(A=@n10, B=Multiply(A=@n10, B=NamedRerouteDeclaration(None=@n12) {name=Min Size, nodeColor=<rgba 0 0 0 1>}), Alpha=@n13)
  n15 = FunctionOutput(None=MaterialFunctionCall(
    UVs=@n1
    Size=NamedRerouteDeclaration(None=@n2) {name=Size, nodeColor=<rgba 0 0 0 1>}
    CornerRoundness=NamedRerouteDeclaration(None=@n14) {name=Corner Radius with Padding, nodeColor=<rgba 0 0 0 1>}
    KeepAspectRatio=StaticBool() {value=true}
    PixelSize=NamedRerouteUsage()
   ) {materialFunction=&MF_UI_Box}.SDF) {description=Output the SDF, outputName=SDF, sortPriority=10}
  FunctionOutput(None=MaterialFunctionCall(DistanceField=@n15) {materialFunction=&MF_UI_AntiAliasing}.Inner) {description=Output the mask, outputName=Shape Mask}
  FunctionOutput(None=@n1) {description=Output the UVs, outputName=UVs, sortPriority=50}
  n16 = FunctionInput() {
   bUsePreviewValueAsDefault=true
   description=Define the thickness of the outline
   inputName=Outline Thickness
   inputType=FunctionInput_Scalar
   sortPriority=20
   previewValue=<0.2 0.2 0.2 0.2>
  }
  n17 = NamedRerouteUsage()
  n18 = MaterialFunctionCall!(DistanceField=@n15, Threshold, Thickness=Multiply(A=Divide(A=@n16, B=@n17), B=NamedRerouteUsage()), Alignment=Constant() {r=-1}, Smoothing) {materialFunction=/FortUILibrary/MaterialFunctions/Shapes/Modifiers/MF_UI_ShapeOutline}
  FunctionOutput(None=MaterialFunctionCall(DistanceField=FunctionOutput(None=@n18.SDFOutline) {description=Output the outline SDF, outputName=Outline SDF, sortPriority=11}) {materialFunction=&MF_UI_AntiAliasing}.Inner) {description=Output the outline mask, outputName=Outline Mask, sortPriority=1}
  n19 = NamedRerouteUsage()
  n20 = ComponentMask(None=@n19) {r=true}
  n21 = ComponentMask(None=@n19) {g=true}
  NamedRerouteDeclaration(None=Max(A=Min(A=@n20, B=@n21)) {constB=0}) {name=Min Pixel Size, nodeColor=<rgba 0 0 0 1>}
  n22 = NamedRerouteUsage()
  n23 = FunctionInput() {
   bUsePreviewValueAsDefault=true
   description=Defines the skew angle
   inputName=Skew Angle
   inputType=FunctionInput_Scalar
   sortPriority=5
   previewValue=<0 0 0 0>
  }
  n24 = Multiply(
   A=ComponentMask(None=MaterialFunctionCall!(Dimensions=@n22) {materialFunction=/FortUILibrary/MaterialFunctions/Basics/MF_UI_GetAspectRatio}.V Ratio) {g=true}
   B=NamedRerouteDeclaration(None=@n23) {name=Skew Angle, nodeColor=<rgba 0 0 0 1>}
  )
  n25 = Constant()
  NamedRerouteDeclaration(None=MaterialFunctionCall!(UV=NamedRerouteUsage(), Skew Angle=AppendVector(A=@n24, B=@n25)) {materialFunction=/FortUILibrary/MaterialFunctions/Basics/MF_UI_Skew}.Skewed UVs) {name=Skewed UVs, nodeColor=<rgba 0 0 0 1>}
  n26 = MaterialFunctionCall() {materialFunction=&GetUserInterfaceUV}
  NamedRerouteDeclaration(None=FunctionInput(Preview=@n26.Normalized UV) {bUsePreviewValueAsDefault=true, description=Define the UVs, inputName=UV, inputType=FunctionInput_Vector2}) {name=UV, nodeColor=<rgba 0 0 0 1>}
  NamedRerouteDeclaration(None=MaterialFunctionCall() {materialFunction=/FortUILibrary/MaterialFunctions/Basics/MF_UI_ViewportScale}.Viewport Scale Factor) {name=Resolution Factor, nodeColor=<rgba 0 0 0 1>}
  n27 = FunctionInput(Preview=@n26.Pixel Size) {bUsePreviewValueAsDefault=true, description=Pixel Size Reference, inputName=Pixel Size, inputType=FunctionInput_Vector2, sortPriority=100}
  NamedRerouteDeclaration(None=@n27) {name=Pixel Size, nodeColor=<rgba 0 0 0 1>}
 ]
}
